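; Backpack-mounted auto-turret attachment that handles the player's turret laser visuals, gameplay cues and aim queries.
; Turret state drives scanner ray visibility and targeting, with timelines blending cone, falloff and scale through custom primitive data and a native visuals update.
; Lasers attach to the backpack muzzle socket and take friendly or hostile colors from the local view target.
; For simulated proxies behind the camera, laser aim substitutes replicated client rotation for stale socket rotation.
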(event Custom|PlayTargetAcquiredGC
  ;[1 1 1] Play Target Acquired GC
  (GameplayCue|ExecuteGameplayCueOnActor(Burst) (Variables|RunTimeVars|GetPlayer%20Pawn) (Variables|GameplayCues|GetGC_TargetAcquired) (Ability|GameplayCue|MakeGameplayCueParameters :NormalizedMagnitude 0.0 :RawMagnitude 0.0 :Location "0,0,0" :Normal "0,0,0" :SourceObject self :GameplayEffectLevel 1 :AbilityLevel 1 :bReplicateLocationWhenUsingMinimalRepProxy false)))

(event Custom|PlayWindupGC
  ;[1 1 1] Play Windup GC
  (if (Utilities|Enum|Equal(Enum) (Variables|RunTimeVars|GetCurrentState) "WindUp")
    (GameplayCue|ExecuteGameplayCueOnActor(Burst) (Variables|RunTimeVars|GetPlayer%20Pawn) (Variables|GameplayCues|GetGC_Windup) (Ability|GameplayCue|MakeGameplayCueParameters :NormalizedMagnitude 0.0 :RawMagnitude 0.0 :Location "0,0,0" :Normal "0,0,0" :SourceObject self :GameplayEffectLevel 1 :AbilityLevel 1 :bReplicateLocationWhenUsingMinimalRepProxy false))))

(event Custom|PlayWindDownGC
  ;[1 1 1] Play Wind Down GC
  (if (Utilities|Enum|Equal(Enum) (Variables|RunTimeVars|GetCurrentState) "WindDown")
    (GameplayCue|ExecuteGameplayCueOnActor(Burst) (Variables|RunTimeVars|GetPlayer%20Pawn) (Variables|GameplayCues|GetGC_WindDown) (Ability|GameplayCue|MakeGameplayCueParameters :NormalizedMagnitude 0.0 :RawMagnitude 0.0 :Location "0,0,0" :Normal "0,0,0" :SourceObject self :GameplayEffectLevel 1 :AbilityLevel 1 :bReplicateLocationWhenUsingMinimalRepProxy false))))

(event Custom|Cancel%20Targeted%20Actor%20Warning
  ;[1 1 1] Cancel Targeted Actor Warning
  (GameplayCue|RemoveGameplayCueOnActor(Looping) (Variables|RunTimeVars|GetPreviousTargetActor) (Variables|GameplayCues|GetGC_TargetActorWarning) (Ability|GameplayCue|MakeGameplayCueParameters :NormalizedMagnitude 0.0 :RawMagnitude 0.0 :Location "0,0,0" :Normal "0,0,0" :SourceObject self :GameplayEffectLevel 1 :AbilityLevel 1 :bReplicateLocationWhenUsingMinimalRepProxy false))
  (GameplayCue|RemoveGameplayCueOnActor(Looping) (Variables|MurkyVoleTurretAttachment|GetTargetActor) (Variables|GameplayCues|GetGC_TargetActorWarning) (Ability|GameplayCue|MakeGameplayCueParameters :NormalizedMagnitude 0.0 :RawMagnitude 0.0 :Location "0,0,0" :Normal "0,0,0" :SourceObject self :GameplayEffectLevel 1 :AbilityLevel 1 :bReplicateLocationWhenUsingMinimalRepProxy false)))

(event Custom|PlayTargetedActorWarningGC
  ;[1 1 1] Play Targeted Actor Warning GC
  (Utilities|IsValid (Variables|MurkyVoleTurretAttachment|GetTargetActor)
    (:Is%20Valid
      (Utilities|Casting|CastToFortPlayerPawn (Variables|MurkyVoleTurretAttachment|GetTargetActor))
      (GameplayCue|AddGameplayCueOnActor(Looping) (Variables|MurkyVoleTurretAttachment|GetTargetActor) (Variables|GameplayCues|GetGC_TargetActorWarning) (Ability|GameplayCue|MakeGameplayCueParameters :NormalizedMagnitude 0.0 :RawMagnitude 0.0 :Location "0,0,0" :Normal "0,0,0" :SourceObject self :GameplayEffectLevel 1 :AbilityLevel 1 :bReplicateLocationWhenUsingMinimalRepProxy false)))))

(event Custom|Play%20EMP%20GC
  ;[1 1 1] Play EMP GC
  (GameplayCue|AddGameplayCueOnActor(Looping) (Variables|RunTimeVars|GetCurrentOwner) (Variables|TimerHandles,Tasks,andTags|GetGC%20EMP) (Ability|GameplayCue|MakeGameplayCueParameters :NormalizedMagnitude 0.0 :RawMagnitude 0.0 :Location "0,0,0" :Normal "0,0,0" :SourceObject self :GameplayEffectLevel 1 :AbilityLevel 1 :bReplicateLocationWhenUsingMinimalRepProxy false)))

(event Custom|Cancel%20EMP%20GC
  ;[1 1 1] Cancel EMP GC
  (GameplayCue|RemoveGameplayCueOnActor(Looping) (Variables|RunTimeVars|GetCurrentOwner) (Variables|TimerHandles,Tasks,andTags|GetGC%20EMP) (Ability|GameplayCue|MakeGameplayCueParameters :NormalizedMagnitude 0.0 :RawMagnitude 0.0 :Location "0,0,0" :Normal "0,0,0" :SourceObject self :GameplayEffectLevel 1 :AbilityLevel 1 :bReplicateLocationWhenUsingMinimalRepProxy false)))

(event Custom|StopAllSoundsAndVFX
  ;[1 1 1] Stop All Sounds and VFX
  (CallFunction|SetScannerRaysTargeting false)
  (CallFunction|SetScannerRaysActive false))

(event Custom|UpdateScannerRays
  ;[1 1 1] Update Scanner Rays
  (switch Utilities|FlowControl|Switch|SwitchonEMurkyVoleTurretState (Variables|RunTimeVars|GetCurrentState)
    (:Deactivated
      _)
    (:Activating
      _)
    (:Scanning
      (CallFunction|SetScannerRaysActive true :!label _label)
      (CallFunction|SetScannerRaysTargeting false))
    (:Tracking
      (goto _label))
    (:Attacking
      (CallFunction|SetScannerRaysTargeting true :!label _label_2))
    (:WindUp
      (goto _label_2))
    (:WindDown
      (goto _label))
    (:ManualControl
      _))
  (CallFunction|SetScannerRaysActive false)
  (CallFunction|SetScannerRaysTargeting false))

(event Custom|AddLaser
  ;[1 1 1] Turret Added to Inventory
  (CallFunction|InitializeScanningRaysParameters)
  (CallFunction|MuzzleAttachment)
  (Development|SetHiddeninGame (Variables|Default|GetScannerRays) false true)
  (Development|SetHiddeninGame (Variables|Default|GetIdle%20Laser) false true))

(event Custom|RemoveLaser
  ;[1 1 1] Turret Removed From Invetory
  (CallFunction|MuzzleAttachment)
  (Development|SetHiddeninGame (Variables|Default|GetScannerRays) true true)
  (Development|SetHiddeninGame (Variables|Default|GetIdle%20Laser) true true))

(event Custom|Muzzle%20Attachment
  ;[1 1 1] Turret
  ;;[1 1 1] Muzzle Attachment
  (if (and (Utilities|IsValid (Variables|RunTimeVars|GetCCPM%20Athena%20Murky%20Vole%20Back%20Pack) :!class /Script/Engine.KismetSystemLibrary) (GameplayTags|HasAnyMatchingGameplayTags (Variables|RunTimeVars|GetPlayer%20Pawn) (Variables|TimerHandles,Tasks,andTags|GetTurret%20In%20Inventory%20Tag%20Container)))
    (Transformation|AttachComponentToComponent (Variables|Default|GetScannerRays) (CustomCharacter|GetSkeletalMeshForPartOnPlayerPawn (Variables|RunTimeVars|GetCCPM%20Athena%20Murky%20Vole%20Back%20Pack)) (Variables|Balance|SystemDefaults|GetMuzzleSocket) "SnapToTarget" "SnapToTarget" "KeepRelative" true)
    (Transformation|AttachComponentToComponent (Variables|Default|GetScanningLoopAudio) (CustomCharacter|GetSkeletalMeshForPartOnPlayerPawn (Variables|RunTimeVars|GetCCPM%20Athena%20Murky%20Vole%20Back%20Pack)) (Variables|Balance|SystemDefaults|GetMuzzleSocket) "SnapToTarget" "KeepRelative" "KeepRelative" true)
    (Transformation|AttachComponentToComponent (Variables|Default|GetIdle%20Laser) (CustomCharacter|GetSkeletalMeshForPartOnPlayerPawn (Variables|RunTimeVars|GetCCPM%20Athena%20Murky%20Vole%20Back%20Pack)) (Variables|Balance|SystemDefaults|GetMuzzleSocket) "SnapToTarget" "KeepRelative" "KeepRelative" true)
    (else
      (Transformation|DetachFromComponent (Variables|Default|GetIdle%20Laser) "KeepRelative" "KeepRelative" "KeepRelative" true)
      (Transformation|DetachFromComponent (Variables|Default|GetScannerRays) "KeepRelative" "KeepRelative" "KeepRelative" true)
      (Transformation|DetachFromComponent (Variables|Default|GetScanningLoopAudio) "KeepRelative" "KeepRelative" "KeepRelative" true))))

(event Custom|InitializeScanningRaysParameters
  (bind _player_pawn (Variables|RunTimeVars|GetPlayer%20Pawn))
  (bind _returnvalue (Fortnite|GetFirstLocalFortPlayerController true true))
  (bind _returnvalue_2 (Pawn|GetViewTarget _returnvalue))
  (bind _returnvalue_3 (Team|GetTeamAffiliation _returnvalue_2 _player_pawn))
  (bind _returnvalue_4 (Utilities|Enum|Equal(Enum) _returnvalue_3 "Friendly"))
  (bind (_targeting_laseredgecolor_friendly_r _targeting_laseredgecolor_friendly_g _targeting_laseredgecolor_friendly_b _targeting_laseredgecolor_friendly_a) (Math|Color|BreakColor (Variables|GetTargeting_LaserEdgeColor_Friendly_R)))
  (bind _scannerrays (Variables|Default|GetScannerRays))
  (bind (_targeting_lasercentercolor_hostile_r _targeting_lasercentercolor_hostile_g _targeting_lasercentercolor_hostile_b _targeting_lasercentercolor_hostile_a) (Math|Color|BreakColor (Variables|GetTargeting_LaserCenterColor_Hostile_R)))
  (bind (_targeting_lasercentercolor_friendly_r _targeting_lasercentercolor_friendly_g _targeting_lasercentercolor_friendly_b _targeting_lasercentercolor_friendly_a) (Math|Color|BreakColor (Variables|GetTargeting_LaserCenterColor_Friendly_R)))
  (bind (_targeting_laseredgecolor_hostile_r _targeting_laseredgecolor_hostile_g _targeting_laseredgecolor_hostile_b _targeting_laseredgecolor_hostile_a) (Math|Color|BreakColor (Variables|GetTargeting_LaserEdgeColor_Hostile_R)))
  ;[1 1 1] Turret
  ;;[1 1 1] Scanner Rays and Beacon Particles - Initialize Parameters
  (Networking|CanExecuteCosmeticEvents
    (:True
      (Rendering|Material|SetCustomPrimitiveDataFloat _scannerrays 0 (Variables|Laser|Size|GetScannerRays_ConeAngle))
      (Rendering|Material|SetCustomPrimitiveDataFloat _scannerrays 1 (Variables|Laser|Size|GetScannerRays_LaserEdgeWidthEnd))
      (Rendering|Material|SetCustomPrimitiveDataFloat _scannerrays 2 (Variables|Laser|Size|GetScannerRays_LaserEdgeWidthFront))
      (Rendering|Material|SetCustomPrimitiveDataFloat _scannerrays 3 (Variables|Laser|Size|GetScannerRays_LaserFalloff))
      (Rendering|Material|SetCustomPrimitiveDataVector4 _scannerrays 4 (select _returnvalue_4 (Math|Vector4|MakeVector4 _targeting_laseredgecolor_friendly_r _targeting_laseredgecolor_friendly_g _targeting_laseredgecolor_friendly_b _targeting_laseredgecolor_friendly_a) (Math|Vector4|MakeVector4 _targeting_laseredgecolor_hostile_r _targeting_laseredgecolor_hostile_g _targeting_laseredgecolor_hostile_b _targeting_laseredgecolor_hostile_a)))
      (Rendering|Material|SetCustomPrimitiveDataVector4 _scannerrays 8 (select _returnvalue_4 (Math|Vector4|MakeVector4 _targeting_lasercentercolor_friendly_r _targeting_lasercentercolor_friendly_g _targeting_lasercentercolor_friendly_b _targeting_lasercentercolor_friendly_a) (Math|Vector4|MakeVector4 _targeting_lasercentercolor_hostile_r _targeting_lasercentercolor_hostile_g _targeting_lasercentercolor_hostile_b _targeting_lasercentercolor_hostile_a)))
      (Rendering|Material|SetCustomPrimitiveDataFloat _scannerrays 12 (Variables|Laser|Size|GetScannerRays_SecondaryRaysSpeed))
      (Rendering|Material|SetCustomPrimitiveDataFloat _scannerrays 13 (Variables|Laser|Size|GetScannerRays_SecondaryRaysOpacity))
      (Rendering|Material|SetScalarParameterValueonMaterials (Variables|Default|GetIdle%20Laser) "Overall Opacity" 1.0)
      (Variables|Laser|Size|SetScannerRays_StartingScale (Variables|/Script/Engine.SceneComponent|GetRelativeScale3D (Variables|Default|GetScannerRays))))))

(event Custom|SetScannerRaysTargeting
  (params (IsTargeting :type Boolean))
  (bind _ccpm_athena_murky_vole_back_pack (Variables|RunTimeVars|GetCCPM%20Athena%20Murky%20Vole%20Back%20Pack))
  ;[1 1 1] Turret
  ;;[1 0.2955 0.2257] Set Scanner Rays Targeting
  (Networking|CanExecuteCosmeticEvents
    (:True
      (if IsTargeting
        (CallFunction|UpdateTargetingLaserScale false true)
        (CallFunction|SetTargetingLaserActive true)
        (Class|CCPMAthenaMurkyVoleBackPack|SetAnimAlwaysTick _ccpm_athena_murky_vole_back_pack true)
        (Utilities|DoOnceThenDo false
          (:DoOnce
            (bind (_direction _alpha) (Timeline|Timeline PlayFromStart)
              (:Update
                (bind _structout (Utilities|Struct|SetmembersinMurkyVoleTurretAttachmentUpdateLaserVisualsParams (Variables|Laser|Config|GetUpdateLaserVisualsParams) (Variables|Laser|Size|GetScannerRays_StartingScale) (Variables|MurkyVoleTurretAttachment|GetScannerRays_TargetingScale) _alpha))
                (CallFunction|NativeUpdateLaserVisuals _structout (Variables|RunTimeVars|GetPlayer%20Pawn) (Variables|Default|GetScannerRays) (Variables|Default|GetIdle%20Laser)))))
          (:ThenDo
            (Timeline|Timeline Play)))
        (else
          (CallFunction|UpdateTargetingLaserScale false true)
          (CallFunction|SetTargetingLaserActive false)
          (Class|CCPMAthenaMurkyVoleBackPack|SetAnimAlwaysTick _ccpm_athena_murky_vole_back_pack false)
          (Utilities|DoOnceThenDo false
            (:DoOnce
              (Timeline|Timeline ReverseFromEnd))
            (:ThenDo
              (Timeline|Timeline Reverse))))))))

(event Custom|OnViewTargetChanged_Event
  (params (PlayerController :type Fort%20Player%20Controller%20Object%20Reference) (OldViewTarget :type Actor%20Object%20Reference) (NewViewTarget :type Actor%20Object%20Reference))
  ;[1 1 1] Turret
  ;;[1 0.2955 0.2257] Set Scanner Rays Targeting
  (if (not (== NewViewTarget (Variables|RunTimeVars|GetPlayer%20Pawn)))
    (Utilities|DoOnceThenDo false
      (:DoOnce
        (Timeline|Timeline ReverseFromEnd))
      (:ThenDo
        (Timeline|Timeline Reverse)))))

(event Custom|SetTargetingLaserActive
  (params (bIsActive :type Boolean))
  (bind _returnvalue (Utilities|Time|GetGameTimeinSeconds))
  (bind _idle_laser (Variables|Default|GetIdle%20Laser))
  ;[1 1 1] Turret
  ;;[0.2902 0.9804 1] Set Targeting Laser Active
  (Rendering|Material|SetCustomPrimitiveDataFloat (Variables|Default|GetScannerRays) 3 (Math|Float|SelectFloat 1.0 (Variables|Laser|Size|GetScannerRays_LaserFalloff) bIsActive))
  (if bIsActive
    (Rendering|Material|SetScalarParameterValueonMaterials _idle_laser "Time Stamp A" _returnvalue)
    (Utilities|FlowControl|Delay 0.05)
    (Rendering|Material|SetScalarParameterValueonMaterials _idle_laser "Time Stamp B" _returnvalue)))

(event Custom|SetScannerRaysActive
  (params (SetActive :type Boolean))
  (bind _currentstate (Variables|RunTimeVars|GetCurrentState))
  (bind _returnvalue (Utilities|Enum|Equal(Enum) _currentstate "ManualControl"))
  (bind _returnvalue_2 (Utilities|Enum|Equal(Enum) _currentstate "Activating"))
  (bind _returnvalue_3 (Utilities|Enum|Equal(Enum) _currentstate "Deactivated"))
  (bind _returnvalue_4 (Math|Boolean|ORBoolean _returnvalue_3 _returnvalue_2 _returnvalue))
  ;[1 1 1] Turret
  ;;[0.2902 0.9804 1] Set Scanner Rays Active - Was Set Turret Hidden State
  (Networking|CanExecuteCosmeticEvents
    (:True
      (if (not SetActive)
        (Utilities|DoOnceThenDo false
          (:DoOnce
            (bind (_direction _animation _raylength _rayangle) (Timeline|Timeline_0 PlayFromStart)
              (:Update
                (if (Math|Boolean|ANDBoolean (Utilities|Enum|NotEqual(Enum) (Variables|RunTimeVars|GetCurrentState) "Attacking") (Utilities|Enum|NotEqual(Enum) (Variables|RunTimeVars|GetCurrentState) "Tracking") (Utilities|Enum|NotEqual(Enum) (Variables|RunTimeVars|GetCurrentState) "WindUp"))
                  (Rendering|Material|SetCustomPrimitiveDataFloat (Variables|Default|GetScannerRays) 3 (Math|Float|Lerp 0.0 (Variables|Laser|Size|GetScannerRays_Laser_Falloff) _raylength))
                  (Rendering|Material|SetCustomPrimitiveDataFloat (Variables|Default|GetScannerRays) 0 (Math|Float|Lerp 0.0 (Variables|Laser|Size|GetScannerRays_ConeAngle) _rayangle))
                  (Rendering|Material|SetScalarParameterValueonMaterials (Variables|Default|GetIdle%20Laser) "Overall Opacity" (Math|Float|Lerp 1.0 1.0 _rayangle))))
              (:Finished
                (switch Utilities|FlowControl|Switch|SwitchonETimelineDirection _direction
                  (:Forward
                    (Development|SetHiddeninGame (Variables|Default|GetScannerRays) _returnvalue_4 false))))
              (:SetRayHidden
                (Development|SetHiddeninGame (Variables|Default|GetScannerRays) _returnvalue_4 false))))
          (:ThenDo
            (Timeline|Timeline_0 Play)))
        (else
          (Utilities|DoOnceThenDo false
            (:DoOnce
              (Timeline|Timeline_0 ReverseFromEnd))
            (:ThenDo
              (Timeline|Timeline_0 Reverse))))))))

(event Custom|UpdateTargetingLaserScale
  (params (SetMeshScale :type Boolean) (CalculateNow :type Boolean))
  (bind (_scannerrays_targetingscale_x _scannerrays_targetingscale_y _scannerrays_targetingscale_z) (Math|Vector|BreakVector (Variables|GetScannerRays_TargetingScale_X)))
  (bind _scannerrays (Variables|Default|GetScannerRays))
  (bind _relativescale3d (Variables|/Script/Engine.SceneComponent|GetRelativeScale3D _scannerrays))
  (bind _scannerrays_targetingscale (Variables|MurkyVoleTurretAttachment|GetScannerRays_TargetingScale))
  ;[1 1 1] Turret
  ;;[0.2902 0.9804 1] Update Targeting Laser Scale
  (Networking|CanExecuteCosmeticEvents
    (:True
      (if (or CalculateNow (> (Utilities|Time|GetGameTimeinSeconds) (+ (Variables|MurkyVoleTurretAttachment|GetScannerRays_LastScaleUpdateTime) (Variables|MurkyVoleTurretAttachment|GetScannerRays_ScaleUpdateInterval))))
        (Variables|MurkyVoleTurretAttachment|SetScannerRays_LastScaleUpdateTime (Utilities|Time|GetGameTimeinSeconds))
        (bind (_aimsourcelocation _aimsourcerotation) (CallFunction|NativeGetLaserAimSourceLocationandRotation (Variables|RunTimeVars|GetCurrentOwner) (Fortnite|GetFirstLocalFortPlayerController true true) (Variables|MurkyVoleTurretAttachment|GetAimRotation_Client)))
        (bind _returnvalue (Visuals|SentryCalculateTargetingLaserScale self (Math|Transform|MakeTransform _aimsourcelocation _aimsourcerotation) (Transformation|GetWorldLocation (Variables|Default|GetScannerRays)) (ScalableFloat|GetValueatLevel (Variables|MurkyVoleTurretAttachment|GetRow_LaserMaxRange)) (ScalableFloat|GetValueatLevel (Variables|MurkyVoleTurretAttachment|GetRow_LaserDistanceMulti))))
        (Variables|SetScannerRays_TargetingScale_X (Math|Vector|MakeVector _returnvalue _scannerrays_targetingscale_y _scannerrays_targetingscale_z)))
      (if (and SetMeshScale (Utilities|Operators|NotEqual(!=) _relativescale3d _scannerrays_targetingscale 0.01))
        (Transformation|SetRelativeScale3D (Variables|Default|GetScannerRays) (Math|Interpolation|VInterpTo _relativescale3d _scannerrays_targetingscale (Utilities|Time|GetWorldDeltaSeconds) (Variables|MurkyVoleTurretAttachment|GetScannerRays_ScaleUpdateInterpSpeed)))))))

(fn GetAimLocation
  (if (Variables|MurkyVoleTurretAttachment|GetbTargetIsObstructed)
    (return (Variables|MurkyVoleTurretAttachment|GetTargetActorLastKnownLocation))
    (else
      (Utilities|IsValid (Variables|MurkyVoleTurretAttachment|GetTargetActor)
        (:Is%20Valid
          (return (Default|GetAimLocationforActor (Variables|MurkyVoleTurretAttachment|GetTargetActor))))
        (:Is%20Not%20Valid
          (return "0, 0, 0"))))))

(fn GetAimLocationForActor
  (params (Actor :type Actor%20Object%20Reference))
  ;[1 0 0] Note - This is duplicated from the same function in GA_Sentry_EventHandler - TODO - Find common source of this data to pass through
  (bind _aimlocation (Default|SentryGetAimLocationforTargetActor Actor (Variables|MurkyVoleTurretAttachment|GetRow_PlayerPctCapsuleHalfHeightOffset_Default) (Variables|MurkyVoleTurretAttachment|GetPlayerPawnTargetTagToPctCapsuleHalfHeightZOffset) (Variables|MurkyVoleTurretAttachment|GetNonPlayerPawnTargetTagToZOffsetMap)))
  (return _aimlocation))

(fn GetAimSourceLocation
  (bind _socket_backpackattach (Variables|Balance|SystemDefaults|GetSocket_BackpackAttach))
  (bind _player_pawn (Variables|RunTimeVars|GetPlayer%20Pawn))
  (bind _mesh (Variables|/Script/Engine.Character|GetMesh _player_pawn))
  (bind _returnvalue (Transformation|GetSocketRotation _mesh _socket_backpackattach))
  (bind _returnvalue_2 (Math|Vector|GetForwardVector _returnvalue))
  (bind _returnvalue_3 (Math|Vector|GetRightVector _returnvalue))
  (bind _returnvalue_4 (Math|Vector|GetUpVector _returnvalue))
  (bind _socket_backpackattach_2 (Variables|Balance|SystemDefaults|GetSocket_BackpackAttach))
  (bind _player_pawn_2 (Variables|RunTimeVars|GetPlayer%20Pawn))
  (bind _mesh_2 (Variables|/Script/Engine.Character|GetMesh _player_pawn_2))
  (bind _returnvalue_5 (Transformation|GetSocketLocation _mesh_2 _socket_backpackattach_2))
  (bind (_meshcomponent _socketlocation) (Class|CCPMAthenaMurkyVoleBackPack|GetSocketLocationandRotation (Variables|RunTimeVars|GetCCPM%20Athena%20Murky%20Vole%20Back%20Pack) (Variables|Balance|SystemDefaults|GetMuzzleSocket)))
  (bind _vector_aimsourcelocationoffset (Variables|MurkyVoleTurretAttachment|GetVector_AimSourceLocationOffset))
  (bind _vector_aimsourcelocationoffset_crouch (Variables|MurkyVoleTurretAttachment|GetVector_AimSourceLocationOffset_Crouch))
  ;[0 0 1 0.5] Nativized, for reference only
  (Networking|SwitchHasAuthority
    (:Authority
      (bind _returnvalue_6 (GameplayTags|HasMatchingGameplayTag(Message) (Variables|RunTimeVars|GetCurrentOwner) "(TagName=\"Gameplay.Action.Player.Crouch\")"))
      (bind (_returnvalue_x _returnvalue_y _returnvalue_z) (Math|Vector|BreakVector (select _returnvalue_6 _vector_aimsourcelocationoffset_crouch _vector_aimsourcelocationoffset)))
      (Rendering|Debug|DrawDebugSphere (Utilities|Operators|Add _returnvalue_5 (* _returnvalue_2 _returnvalue_x) (* _returnvalue_3 _returnvalue_y) (* _returnvalue_4 _returnvalue_z)) 1e+01 12 "(R=1.000000,G=1.000000,B=1.000000,A=1.000000)" 0.1 0.1)
      (return (Utilities|Operators|Add _returnvalue_5 (* _returnvalue_2 _returnvalue_x) (* _returnvalue_3 _returnvalue_y) (* _returnvalue_4 _returnvalue_z))))
    (:Remote
      (Utilities|IsValid (Variables|RunTimeVars|GetCCPM%20Athena%20Murky%20Vole%20Back%20Pack)
        (:Is%20Valid
          (return _socketlocation))
        (:Is%20Not%20Valid
          (return "0, 0, 0"))))))

(fn GetLaserAimSourceLocationAndRotation
  (bind (_meshcomponent _socketlocation _socketrotation) (Class|CCPMAthenaMurkyVoleBackPack|GetSocketLocationandRotation (Variables|RunTimeVars|GetCCPM%20Athena%20Murky%20Vole%20Back%20Pack) (Variables|Balance|SystemDefaults|GetMuzzleSocket)))
  (bind _muzzlesocket (Variables|Balance|SystemDefaults|GetMuzzleSocket))
  (bind _ccpm_athena_murky_vole_back_pack (Variables|RunTimeVars|GetCCPM%20Athena%20Murky%20Vole%20Back%20Pack))
  (bind (_meshcomponent_2 _socketlocation_2 _socketrotation_2) (Class|CCPMAthenaMurkyVoleBackPack|GetSocketLocationandRotation _ccpm_athena_murky_vole_back_pack _muzzlesocket))
  (bind _returnvalue (Fortnite|GetFirstLocalFortPlayerController true true))
  (Networking|SwitchHasAuthority
    (:Remote
      (if (Utilities|Enum|Equal(Enum) (Networking|GetLocalRole (Variables|RunTimeVars|GetCurrentOwner)) "ROLE_SimulatedProxy")
        ;[0.15 0.15 0.15 0.5 48] Use the calculated aim rotation instead of the socket's rotation when dealing with a sim proxy's turret laser that the local player is _not_ facing. The socket rotation will stop updating when the player looks away from the sim proxy with the turret. The socket's location should still be mostly accurate.
        (Utilities|IsValid _returnvalue
          (:Is%20Valid
            (return _socketlocation_2 (select (> (Math|Vector|DotProduct (Math|Vector|GetForwardVector (Camera|GetCameraRotation (Variables|/Script/Engine.PlayerController|GetPlayerCameraManager _returnvalue))) (- (Transformation|GetActorLocation (Variables|RunTimeVars|GetCurrentOwner)) (Camera|GetCameraLocation (Variables|/Script/Engine.PlayerController|GetPlayerCameraManager _returnvalue)))) 0) _socketrotation_2 (Variables|MurkyVoleTurretAttachment|GetAimRotation_Client))))
          (:Is%20Not%20Valid
            (return _socketlocation_2 (Variables|MurkyVoleTurretAttachment|GetAimRotation_Client))))
        (else
          (return _socketlocation _socketrotation))))))

(fn GetMurkyVoleTurretAttachmentData
  (return (Utilities|Enum|Equal(Enum) (Variables|RunTimeVars|GetCurrentState) "WindUp") (Utilities|Enum|Equal(Enum) (Variables|RunTimeVars|GetCurrentState) "WindDown") (or (Utilities|Enum|Equal(Enum) (Variables|RunTimeVars|GetCurrentState) "ManualControl") (Utilities|Enum|Equal(Enum) (Variables|RunTimeVars|GetCurrentState) "Deactivated")) (Utilities|Select :Index (Variables|RunTimeVars|GetCurrentState) :Deactivated (Variables|MurkyVoleTurretAttachment|GetAimRotation_Client) :Activating (Variables|MurkyVoleTurretAttachment|GetAimRotation_Server) :Scanning (Variables|MurkyVoleTurretAttachment|GetAimRotation_Server) :Tracking (Variables|MurkyVoleTurretAttachment|GetAimRotation_Client) :Attacking (Variables|MurkyVoleTurretAttachment|GetAimRotation_Client) :WindUp (Variables|MurkyVoleTurretAttachment|GetAimRotation_Client) :WindDown (Variables|MurkyVoleTurretAttachment|GetAimRotation_Server) :ManualControl (Variables|MurkyVoleTurretAttachment|GetAimRotation_Client))))

(fn GetMuzzleSocketLocationAndRotation
  (bind (_meshcomponent _socketlocation _socketrotation) (Class|CCPMAthenaMurkyVoleBackPack|GetSocketLocationandRotation (Variables|RunTimeVars|GetCCPM%20Athena%20Murky%20Vole%20Back%20Pack) (Variables|Balance|SystemDefaults|GetMuzzleSocket)))
  (if (Utilities|IsValid (Variables|RunTimeVars|GetCCPM%20Athena%20Murky%20Vole%20Back%20Pack) :!class /Script/Engine.KismetSystemLibrary)
    (return _socketlocation _socketrotation)
    (else
      (return "0, 0, 0" "0, 0, 0"))))

(fn GetTurretBlueprintData
  (bind (_meshcomponent _socketlocation) (Class|CCPMAthenaMurkyVoleBackPack|GetSocketLocationandRotation (Variables|RunTimeVars|GetCCPM%20Athena%20Murky%20Vole%20Back%20Pack) (Variables|Balance|SystemDefaults|GetMuzzleSocket)))
  (Utilities|IsValid (Variables|RunTimeVars|GetCCPM%20Athena%20Murky%20Vole%20Back%20Pack)
    (:Is%20Valid
      (return (Variables|RunTimeVars|GetCurrentOwner) (Variables|Default|GetScannerRays) (Variables|Default|GetIdle%20Laser) (Variables|RunTimeVars|GetCurrentState) _socketlocation))
    (:Is%20Not%20Valid
      (return (Variables|RunTimeVars|GetCurrentOwner) (Variables|Default|GetScannerRays) (Variables|Default|GetIdle%20Laser) (Variables|RunTimeVars|GetCurrentState) "0, 0, 0"))))

(fn InitLaserVisualsParams
  (Variables|Laser|Config|SetUpdateLaserVisualsParams (Utilities|Struct|MakeMurkyVoleTurretAttachmentUpdateLaserVisualsParams "Overall Opacity" (Variables|Laser|Color|GetScanning_Laser_EdgeColor_Hostile) (Variables|Laser|Color|GetScanning_Laser_EdgeColor_Friendly) (Variables|Laser|Color|GetTargeting_LaserEdgeColor_Hostile) (Variables|Laser|Color|GetTargeting_LaserEdgeColor_Friendly) (Variables|Laser|Color|GetTargeting_LaserCenterColor_Hostile) (Variables|Laser|Color|GetTargeting_LaserCenterColor_Friendly) (Variables|Laser|Size|GetScannerRays_ConeAngle) (Variables|Laser|Size|GetScannerRays_LaserEdgeWidthEnd) (Variables|Laser|Size|GetScannerRays_LaserEdgeWidthEnd_Targeting) (Variables|Laser|Size|GetScannerRays_LaserEdgeWidthFront) (Variables|Laser|Size|GetScannerRays_LaserEdgeWidthFront_Targeting) (Variables|Laser|Size|GetScannerRays_Laser_Falloff) (Variables|Laser|Size|GetScannerRays_LaserFalloff_Targeting) 1.0 1.0 0.0 (Variables|Laser|Size|GetScannerRays_StartingScale) (Variables|MurkyVoleTurretAttachment|GetScannerRays_TargetingScale) 1.0)))

(fn ConstructionScript)
