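; Gameplay ability that toggles the MurkyVole turret's manual targeting mode for the player pawn.
; On activation it prints a debug string, commits, then checks the avatar pawn for the ManualMode granted tag.
; Without the tag it applies the manual toggle GameplayEffect and fires the ToggleOn cue; otherwise it removes effects granting that tag and fires ToggleOff.
; EndAbility runs only on the valid-pawn path.
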(event Ability|EventFailedToActivatePassiveAbility
  (params (ActorInfo :type Gameplay%20Ability%20Actor%20Info%20Structure)))

(event Ability|EventActivateAbility
  (Development|PrintString "Toggle Auto Target")
  (Ability|CommitAbility)
  (bind _outfortplayerpawn (CallFunction|GetAvatarasValidFortPlayerPawn)
    (:ValidFortPlayerPawn
      (Variables|Default|SetPlayerPawn _outfortplayerpawn)
      (if (not (GameplayTags|HasMatchingGameplayTag (Variables|Default|GetPlayerPawn) "(TagName=\"Granted.Athena.MurkyVole.Turret.ManualMode\")"))
        (Ability|ApplyGameplayEffectToOwner "/MurkyVoleGameplay/Gameplay/Turret/GE_Athena_MurkyVole_Turret_Manual_Toggle.GE_Athena_MurkyVole_Turret_Manual_Toggle_C")
        (Ability|ExecuteGameplayCueOnOwner (Variables|Default|GetGameplay_Cue_Tag_ToggleOn))
        (else
          (Ability|RemoveGameplayEffectFromOwnerWithGrantedTags "(GameplayTags=((TagName=\"Granted.Athena.MurkyVole.Turret.ManualMode\")))")
          (Ability|ExecuteGameplayCueOnOwner (Variables|Default|GetGameplay_Cue_Tag_ToggleOff))))
      (Ability|EndAbility))))
